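# Editor utility widget for one row of the DelMar Find BR Dependencies track review tool, pairing affected objects with an action.
# AddObject adds unique objects and builds a counted, comma-separated name list, labeled actors or assets by bObjectsInLevel.
# That flag also picks the find button's icon and whether it selects level actors or syncs the Content Browser.
# A second button syncs the Content Browser to assets stored by AddActionAssets, beside the Action text.
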EditorUtilityWidgetBlueprint /DelMarTools/EditorUtilities/AssetReviewTools/TrackPerformanceReview/Subwidgets/FindBRDependencies/FindBRDependenciesToolEntry
graphs [
 AddActionAssets {nodes [
   n1 = AddActionAssets!() -> (then, InAssets:Array of Object References) {is=title K2Node_FunctionEntry}
   > n2 = For Each Loop(Array:Array of Object References=@n1) -> (Array Element:Object Reference)
   > Add Unique(
    TargetArray:Array of Object References=Get ActionAssets() -> (ActionAssets:Array of Object References)
    NewItem:"Object Reference (by ref)"=@n2
   )
  ]}
 AddObject {nodes [
   AddObject!() -> (then, InObject:Object Reference) {is=title K2Node_FunctionEntry}
   > n1 = Add Unique(TargetArray=Get Objects() -> (Objects:Array of Object References), NewItem=Get InObject() -> (InObject:Object Reference))
   > Branch(Condition="integer >= integer"(A=@n1))
   > n2 = Sequence()
   n3 = Get InObject()
   n4 = TextIsEmpty(InText=Get ActorsFormattedText() -> (ActorsFormattedText:Text))
   > Set ActorsFormattedText(ActorsFormattedText:Text=Format Text!(
     Format="INVTEXT(\"{currText}{comma}{actor}\")"
     currText:Text=Get ActorsFormattedText()
     comma:String=SelectString(B=",", bPickA=@n4)
     actor:String=GetDisplayName(Object=@n3)
    ) -> (Result)) -> (Output_Get:Text)
   n5 = Get Objects()
   n6 = Get bObjectsInLevel() -> (bObjectsInLevel:bool)
   "SetText (Editable Text)"!(
    execute=@n2.then_1
    self:Editable Text Object Reference=Get ActorsText() -> (ActorsText:Editor Utility Editable Text Object Reference)
    InText:Text=Format Text!(
     Format="INVTEXT(\"{num} {type}:\\r\\n{actors}\")"
     num:int=Length(TargetArray:Array of Object References=@n5)
     type:String=SelectString(A=actors, B=assets, bPickA=@n6)
     actors:Text=Get ActorsFormattedText()
    ) -> (Result)
   ) -> (then)
  ]}
 EventGraph {nodes [
   "On Clicked (FindObjectsButton)"() {is=title K2Node_ComponentBoundEvent}
   > n1 = Branch(Condition=Get bObjectsInLevel())
   > SelectAffectedLevelActors()
   SelectObjectsInContentBrowser!(execute=@n1.else, self, InObjectsToSelect:Array of Object References=Get Objects()) -> (then)
   "On Clicked (FindActionAssetsButton)"() {is=title K2Node_ComponentBoundEvent}
   > SelectObjectsInContentBrowser(InObjectsToSelect=Get ActionAssets())
   Event PreConstruct()
   > n2 = Sequence()
   > "SetText (Editable Text)"(
    self=Get ActionText() -> (ActionText:Editor Utility Editable Text Object Reference)
    InText="To Text (String)"(InString=Get Action() -> (Action:String))
   )
   SetActiveWidgetIndex(
    execute=@n2.then_1
    self=Get FindObjectsIconSwitcher() -> (FindObjectsIconSwitcher:Widget Switcher Object Reference)
    Index=Select(Option 0:int=1, Option 1:int=0, Index=Get bObjectsInLevel()) -> (ReturnValue:int)
   )
  ]}
 SelectAffectedLevelActors {nodes [
   SelectAffectedLevelActors() {is=title K2Node_FunctionEntry}
   > n1 = Sequence()
   > n2 = For Each Loop(Array=Get Objects())
   > n3 = Cast To Actor!(execute, Object=@n2) -> (then, CastFailed, AsActor:Actor Object Reference) {is=title K2Node_DynamicCast}
   > Add(
    TargetArray:Array of Actor Object References=Get Actors() -> (Actors:Array of Actor Object References)
    NewItem:"Actor Object Reference (by ref)"=@n3
   ) {is=title K2Node_CallArrayFunction}
   SetSelectedLevelActors(execute=@n1.then_1, self=Get EditorActorSubsystem(), ActorsToSelect=Get Actors())
  ]}
 SelectObjectsInContentBrowser {nodes [
   SelectObjectsInContentBrowser!() -> (then, InObjectsToSelect:Array of Object References) {is=title K2Node_FunctionEntry}
   > n1 = Sequence()
   > n2 = For Each Loop(Array=Get InObjectsToSelect() -> (InObjectsToSelect:Array of Object References))
   > Add(TargetArray:Array of Strings=Get AssetPaths() -> (AssetPaths:Array of Strings), NewItem:"String (by ref)"=Get Object Path String(Object=@n2))
   SyncBrowserToObjects(execute=@n1.then_1, AssetPaths=Get AssetPaths())
  ]}
]
variables [
 Action {type=string, default=""}
 ActionAssets {type="array<Object>", default []}
 ActorsFormattedText {type=string, default=""}
 bObjectsInLevel {type=boolean}
 Objects {type="array<Object>", default []}
]
widgets {
 info {rootWidgetClass=&Overlay, widgetCount=22}
 widgets [
  Overlay_123 {widgetClassPath=&Overlay, widget=Overlay_123}
  BottomLine {
   widgetClassPath=&Image
   parent=Overlay_123
   widget=BottomLine
   properties {
    brush {
     bIsDynamicallyLoaded=false
     drawAs=Image
     imageType=NoImage
     mirroring=NoMirror
     resourceName=None
     resourceObject=null
     tiling=NoTile
     imageSize=<32 1>
     margin=<ltrb 0 0 0 0>
     outlineSettings {
      bUseBrushTransparency=false
      roundingType=HalfHeightRadius
      width=0
      color {colorUseRule=UseColor_Specified, specifiedColor=<rgba 0 0 0 0>}
      cornerRadii=<0 0 0 0>
     }
     tintColor {colorUseRule=UseColor_Specified, specifiedColor=<rgba 1 1 1 1>}
     uVRegion {bIsValid=false, max=<0 0>, min=<0 0>}
    }
    colorAndOpacity=<rgba 0.0284 0.0284 0.0284 1>
   }
   slot {class=OverlaySlot, properties {horizontalAlignment=HAlign_Fill, verticalAlignment=VAlign_Bottom}}
  }
  HorizontalBox_126 {
   widgetClassPath=&HorizontalBox
   parent=Overlay_123
   widget=HorizontalBox_126
   slot {class=OverlaySlot, properties {horizontalAlignment=HAlign_Fill, verticalAlignment=VAlign_Fill}}
  }
  Overlay_351 {
   widgetClassPath=&Overlay
   parent=HorizontalBox_126
   widget=Overlay_351
   slot {class=HorizontalBoxSlot, properties {size {sizeRule=Fill, value=1}}}
  }
  VerticalBox_286 {
   widgetClassPath=&VerticalBox
   parent=Overlay_351
   widget=VerticalBox_286
   slot {class=OverlaySlot, properties {horizontalAlignment=HAlign_Fill, verticalAlignment=VAlign_Fill}}
  }
  ActorsText {
   widgetClassPath=/Script/Blutility.EditorUtilityEditableText
   bIsVariable=true
   parent=VerticalBox_286
   widget=ActorsText
   properties {isReadOnly=true, overflowPolicy=Ellipsis}
   slot {class=VerticalBoxSlot, properties {verticalAlignment=VAlign_Top, padding=<ltrb 1 1 1 1>, size {sizeRule=Fill, value=1}}}
  }
  FindObjectsButton {
   widgetClassPath=&Button
   parent=VerticalBox_286
   widget=FindObjectsButton
   properties {
    backgroundColor=<rgba 0 0 0 0.25>
    widgetStyle {
     clickedSlateSound {resourceObject=null}
     disabled {
      bIsDynamicallyLoaded=false
      drawAs=RoundedBox
      imageType=NoImage
      mirroring=NoMirror
      resourceName=None
      resourceObject=null
      tiling=NoTile
      imageSize=<0 0>
      margin=<ltrb 0 0 0 0>
      outlineSettings {
       bUseBrushTransparency=false
       roundingType=FixedRadius
       width=1
       color {colorUseRule=UseColor_ColorTable, specifiedColor=<rgba 0 0 0 0>}
       cornerRadii=<4 4 4 4>
      }
      tintColor {colorUseRule=UseColor_Specified, specifiedColor=<rgba 0.0395 0.0395 0.0395 1>}
      uVRegion {bIsValid=false, max=<0 0>, min=<0 0>}
     }
     disabledForeground {colorUseRule=UseColor_ColorTable, specifiedColor=<rgba 0 0 0 0>}
     hovered {
      bIsDynamicallyLoaded=false
      drawAs=RoundedBox
      imageType=NoImage
      mirroring=NoMirror
      resourceName=None
      resourceObject=null
      tiling=NoTile
      imageSize=<32 32>
      margin=<ltrb 0 0 0 0>
      outlineSettings {
       bUseBrushTransparency=false
       roundingType=FixedRadius
       width=1
       color {colorUseRule=UseColor_ColorTable, specifiedColor=<rgba 0 0 0 0>}
       cornerRadii=<4 4 4 4>
      }
      tintColor {colorUseRule=UseColor_Specified, specifiedColor=<rgba 0.0953 0.0953 0.0953 1>}
      uVRegion {bIsValid=false, max=<0 0>, min=<0 0>}
     }
     hoveredForeground {colorUseRule=UseColor_ColorTable, specifiedColor=<rgba 0 0 0 0>}
     hoveredSlateSound {resourceObject=null}
     normal {
      bIsDynamicallyLoaded=false
      drawAs=RoundedBox
      imageType=NoImage
      mirroring=NoMirror
      resourceName=None
      resourceObject=null
      tiling=NoTile
      imageSize=<32 32>
      margin=<ltrb 0 0 0 0>
      outlineSettings {
       bUseBrushTransparency=false
       roundingType=FixedRadius
       width=1
       color {colorUseRule=UseColor_ColorTable, specifiedColor=<rgba 0 0 0 0>}
       cornerRadii=<4 4 4 4>
      }
      tintColor {colorUseRule=UseColor_Specified, specifiedColor=<rgba 0.0395 0.0395 0.0395 1>}
      uVRegion {bIsValid=false, max=<0 0>, min=<0 0>}
     }
     normalForeground {colorUseRule=UseColor_ColorTable, specifiedColor=<rgba 0 0 0 0>}
     normalPadding=<ltrb 12 1.5 12 1.5>
     pressed {
      bIsDynamicallyLoaded=false
      drawAs=RoundedBox
      imageType=NoImage
      mirroring=NoMirror
      resourceName=None
      resourceObject=null
      tiling=NoTile
      imageSize=<32 32>
      margin=<ltrb 0 0 0 0>
      outlineSettings {
       bUseBrushTransparency=false
       roundingType=FixedRadius
       width=1
       color {colorUseRule=UseColor_ColorTable, specifiedColor=<rgba 0 0 0 0>}
       cornerRadii=<4 4 4 4>
      }
      tintColor {colorUseRule=UseColor_Specified, specifiedColor=<rgba 0.0284 0.0284 0.0284 1>}
      uVRegion {bIsValid=false, max=<0 0>, min=<0 0>}
     }
     pressedForeground {colorUseRule=UseColor_ColorTable, specifiedColor=<rgba 0 0 0 0>}
     pressedPadding=<ltrb 12 2.5 12 0.5>
     pressedSlateSound {resourceObject=null}
    }
   }
   slot {class=VerticalBoxSlot, properties {horizontalAlignment=HAlign_Right, verticalAlignment=VAlign_Bottom}}
  }
  FindObjectsButtonIconSizeBox {
   widgetClassPath=&SizeBox
   parent=FindObjectsButton
   widget=FindObjectsButtonIconSizeBox
   properties {bOverride_HeightOverride=true, bOverride_WidthOverride=true, heightOverride=28, widthOverride=28}
   slot {class=ButtonSlot, properties {padding=<ltrb 0 0 0 0>}}
  }
  Overlay {widgetClassPath=&Overlay, parent=FindObjectsButtonIconSizeBox, widget=Overlay, slot {class=SizeBoxSlot}}
  Image_1 {
   widgetClassPath=&Image
   parent=Overlay
   widget=Image_1
   properties {
    brush {
     bIsDynamicallyLoaded=false
     drawAs=Image
     imageType=NoImage
     mirroring=NoMirror
     resourceName=None
     resourceObject=/Game/UI/Friends_UI/Social/Search
     tiling=NoTile
     imageSize=<20 20>
     margin=<ltrb 0 0 0 0>
     outlineSettings {
      bUseBrushTransparency=false
      roundingType=HalfHeightRadius
      width=0
      color {colorUseRule=UseColor_Specified, specifiedColor=<rgba 0 0 0 0>}
      cornerRadii=<0 0 0 0>
     }
     tintColor {colorUseRule=UseColor_Specified, specifiedColor=<rgba 1 1 1 1>}
     uVRegion {bIsValid=false, max=<0 0>, min=<0 0>}
    }
    colorAndOpacity=<rgba 0.3906 0.3906 0.3906 1>
   }
   slot {class=OverlaySlot, properties {horizontalAlignment=HAlign_Right}}
  }
  FindObjectsIconSwitcher {
   widgetClassPath=&WidgetSwitcher
   bIsVariable=true
   parent=Overlay
   widget=FindObjectsIconSwitcher
   slot {class=OverlaySlot, properties {verticalAlignment=VAlign_Bottom}}
  }
  Image_2 {
   widgetClassPath=&Image
   parent=FindObjectsIconSwitcher
   widget=Image_2
   properties {
    clipping=ClipToBounds
    brush {
     bIsDynamicallyLoaded=false
     drawAs=Image
     imageType=NoImage
     mirroring=NoMirror
     resourceName=None
     resourceObject=/Engine/Tutorial/Landscape/TutorialAssets/Landscape
     tiling=NoTile
     imageSize=<20 18>
     margin=<ltrb 0 0 0 0>
     outlineSettings {
      bUseBrushTransparency=false
      roundingType=HalfHeightRadius
      width=0
      color {colorUseRule=UseColor_Specified, specifiedColor=<rgba 0 0 0 0>}
      cornerRadii=<0 0 0 0>
     }
     tintColor {colorUseRule=UseColor_Specified, specifiedColor=<rgba 1 1 1 1>}
     uVRegion {bIsValid=false, max=<0 0>, min=<0 0>}
    }
    renderTransform {angle=0, scale=<1 1>, shear=<0 0>, translation=<0 1>}
   }
   slot {class=WidgetSwitcherSlot, properties {horizontalAlignment=HAlign_Left, verticalAlignment=VAlign_Bottom}}
  }
  Image_3 {
   widgetClassPath=&Image
   parent=FindObjectsIconSwitcher
   widget=Image_3
   properties {
    clipping=ClipToBounds
    brush {
     bIsDynamicallyLoaded=false
     drawAs=Image
     imageType=NoImage
     mirroring=NoMirror
     resourceName=None
     resourceObject=/Game/UI/Foundation/Textures/Icons/Replay/T-Icon-Replay-Folder-L
     tiling=NoTile
     imageSize=<20 20>
     margin=<ltrb 0 0 0 0>
     outlineSettings {
      bUseBrushTransparency=false
      roundingType=HalfHeightRadius
      width=0
      color {colorUseRule=UseColor_Specified, specifiedColor=<rgba 0 0 0 0>}
      cornerRadii=<0 0 0 0>
     }
     tintColor {colorUseRule=UseColor_Specified, specifiedColor=<rgba 1 1 1 1>}
     uVRegion {bIsValid=false, max=<0 0>, min=<0 0>}
    }
    renderTransform {angle=0, scale=<1 1>, shear=<0 0>, translation=<0 3>}
   }
   slot {class=WidgetSwitcherSlot, properties {horizontalAlignment=HAlign_Left, verticalAlignment=VAlign_Bottom}}
  }
  VerticalDivider {
   widgetClassPath=&Image
   parent=HorizontalBox_126
   widget=VerticalDivider
   properties {brush {
     bIsDynamicallyLoaded=false
     drawAs=Image
     imageType=NoImage
     mirroring=NoMirror
     resourceName=None
     resourceObject=null
     tiling=NoTile
     imageSize=<1 32>
     margin=<ltrb 0 0 0 0>
     outlineSettings {
      bUseBrushTransparency=false
      roundingType=HalfHeightRadius
      width=0
      color {colorUseRule=UseColor_Specified, specifiedColor=<rgba 0 0 0 0>}
      cornerRadii=<0 0 0 0>
     }
     tintColor {colorUseRule=UseColor_Specified, specifiedColor=<rgba 0.0284 0.0284 0.0284 1>}
     uVRegion {bIsValid=false, max=<0 0>, min=<0 0>}
    }}
   slot {class=HorizontalBoxSlot, properties {horizontalAlignment=HAlign_Center}}
  }
  Overlay_127 {
   widgetClassPath=&Overlay
   parent=HorizontalBox_126
   widget=Overlay_127
   slot {class=HorizontalBoxSlot, properties {size {sizeRule=Fill, value=1}}}
  }
  VerticalBox_425 {
   widgetClassPath=&VerticalBox
   parent=Overlay_127
   widget=VerticalBox_425
   slot {class=OverlaySlot, properties {horizontalAlignment=HAlign_Fill, verticalAlignment=VAlign_Fill}}
  }
  ActionText {
   widgetClassPath=&EditorUtilityEditableText
   bIsVariable=true
   parent=VerticalBox_425
   widget=ActionText
   properties {isReadOnly=true, overflowPolicy=Ellipsis}
   slot {class=VerticalBoxSlot, properties {verticalAlignment=VAlign_Top, padding=<ltrb 1 1 1 1>, size {sizeRule=Fill, value=1}}}
  }
  FindActionAssetsButton {
   widgetClassPath=&Button
   parent=VerticalBox_425
   widget=FindActionAssetsButton
   properties {
    backgroundColor=<rgba 0 0 0 0.25>
    widgetStyle {
     clickedSlateSound {resourceObject=null}
     disabled {
      bIsDynamicallyLoaded=false
      drawAs=RoundedBox
      imageType=NoImage
      mirroring=NoMirror
      resourceName=None
      resourceObject=null
      tiling=NoTile
      imageSize=<0 0>
      margin=<ltrb 0 0 0 0>
      outlineSettings {
       bUseBrushTransparency=false
       roundingType=FixedRadius
       width=1
       color {colorUseRule=UseColor_ColorTable, specifiedColor=<rgba 0 0 0 0>}
       cornerRadii=<4 4 4 4>
      }
      tintColor {colorUseRule=UseColor_Specified, specifiedColor=<rgba 0.0395 0.0395 0.0395 1>}
      uVRegion {bIsValid=false, max=<0 0>, min=<0 0>}
     }
     disabledForeground {colorUseRule=UseColor_ColorTable, specifiedColor=<rgba 0 0 0 0>}
     hovered {
      bIsDynamicallyLoaded=false
      drawAs=RoundedBox
      imageType=NoImage
      mirroring=NoMirror
      resourceName=None
      resourceObject=null
      tiling=NoTile
      imageSize=<32 32>
      margin=<ltrb 0 0 0 0>
      outlineSettings {
       bUseBrushTransparency=false
       roundingType=FixedRadius
       width=1
       color {colorUseRule=UseColor_ColorTable, specifiedColor=<rgba 0 0 0 0>}
       cornerRadii=<4 4 4 4>
      }
      tintColor {colorUseRule=UseColor_Specified, specifiedColor=<rgba 0.0953 0.0953 0.0953 1>}
      uVRegion {bIsValid=false, max=<0 0>, min=<0 0>}
     }
     hoveredForeground {colorUseRule=UseColor_ColorTable, specifiedColor=<rgba 0 0 0 0>}
     hoveredSlateSound {resourceObject=null}
     normal {
      bIsDynamicallyLoaded=false
      drawAs=RoundedBox
      imageType=NoImage
      mirroring=NoMirror
      resourceName=None
      resourceObject=null
      tiling=NoTile
      imageSize=<32 32>
      margin=<ltrb 0 0 0 0>
      outlineSettings {
       bUseBrushTransparency=false
       roundingType=FixedRadius
       width=1
       color {colorUseRule=UseColor_ColorTable, specifiedColor=<rgba 0 0 0 0>}
       cornerRadii=<4 4 4 4>
      }
      tintColor {colorUseRule=UseColor_Specified, specifiedColor=<rgba 0.0395 0.0395 0.0395 1>}
      uVRegion {bIsValid=false, max=<0 0>, min=<0 0>}
     }
     normalForeground {colorUseRule=UseColor_ColorTable, specifiedColor=<rgba 0 0 0 0>}
     normalPadding=<ltrb 12 1.5 12 1.5>
     pressed {
      bIsDynamicallyLoaded=false
      drawAs=RoundedBox
      imageType=NoImage
      mirroring=NoMirror
      resourceName=None
      resourceObject=null
      tiling=NoTile
      imageSize=<32 32>
      margin=<ltrb 0 0 0 0>
      outlineSettings {
       bUseBrushTransparency=false
       roundingType=FixedRadius
       width=1
       color {colorUseRule=UseColor_ColorTable, specifiedColor=<rgba 0 0 0 0>}
       cornerRadii=<4 4 4 4>
      }
      tintColor {colorUseRule=UseColor_Specified, specifiedColor=<rgba 0.0284 0.0284 0.0284 1>}
      uVRegion {bIsValid=false, max=<0 0>, min=<0 0>}
     }
     pressedForeground {colorUseRule=UseColor_ColorTable, specifiedColor=<rgba 0 0 0 0>}
     pressedPadding=<ltrb 12 2.5 12 0.5>
     pressedSlateSound {resourceObject=null}
    }
   }
   slot {class=VerticalBoxSlot, properties {horizontalAlignment=HAlign_Right, verticalAlignment=VAlign_Bottom}}
  }
  SizeBox_0 {
   widgetClassPath=&SizeBox
   parent=FindActionAssetsButton
   widget=SizeBox_0
   properties {bOverride_HeightOverride=true, bOverride_WidthOverride=true, heightOverride=28, widthOverride=28}
   slot {class=ButtonSlot, properties {padding=<ltrb 0 0 0 0>}}
  }
  Overlay_0 {widgetClassPath=&Overlay, parent=SizeBox_0, widget=Overlay_0, slot {class=SizeBoxSlot}}
  Image_386 {
   widgetClassPath=&Image
   parent=Overlay_0
   widget=Image_386
   properties {
    brush {
     bIsDynamicallyLoaded=false
     drawAs=Image
     imageType=NoImage
     mirroring=NoMirror
     resourceName=None
     resourceObject=&Search
     tiling=NoTile
     imageSize=<20 20>
     margin=<ltrb 0 0 0 0>
     outlineSettings {
      bUseBrushTransparency=false
      roundingType=HalfHeightRadius
      width=0
      color {colorUseRule=UseColor_Specified, specifiedColor=<rgba 0 0 0 0>}
      cornerRadii=<0 0 0 0>
     }
     tintColor {colorUseRule=UseColor_Specified, specifiedColor=<rgba 1 1 1 1>}
     uVRegion {bIsValid=false, max=<0 0>, min=<0 0>}
    }
    colorAndOpacity=<rgba 0.3906 0.3906 0.3906 1>
   }
   slot {class=OverlaySlot, properties {horizontalAlignment=HAlign_Right}}
  }
  Image {
   widgetClassPath=&Image
   parent=Overlay_0
   widget=Image
   properties {
    clipping=ClipToBounds
    brush {
     bIsDynamicallyLoaded=false
     drawAs=Image
     imageType=NoImage
     mirroring=NoMirror
     resourceName=None
     resourceObject=&T-Icon-Replay-Folder-L
     tiling=NoTile
     imageSize=<20 20>
     margin=<ltrb 0 0 0 0>
     outlineSettings {
      bUseBrushTransparency=false
      roundingType=HalfHeightRadius
      width=0
      color {colorUseRule=UseColor_Specified, specifiedColor=<rgba 0 0 0 0>}
      cornerRadii=<0 0 0 0>
     }
     tintColor {colorUseRule=UseColor_Specified, specifiedColor=<rgba 1 1 1 1>}
     uVRegion {bIsValid=false, max=<0 0>, min=<0 0>}
    }
    renderTransform {angle=0, scale=<1 1>, shear=<0 0>, translation=<0 3>}
   }
   slot {class=OverlaySlot, properties {verticalAlignment=VAlign_Bottom}}
  }
 ]
}
