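# Editor utility widget that finds materials in level actors or selected Content Browser assets depending on problematic Fortnite material functions, and suggests fixes.
# Searches step through one actor or asset per short async delay, with progress and cancel, expanding building level instances and recursing through Asset Registry dependencies of relevant types.
# Fixes point to an existing ported M_DM_ duplicate or describe creating one, grouped by action in entry widgets and exportable to CSV via Python.
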EditorUtilityWidgetBlueprint /DelMarTools/EditorUtilities/AssetReviewTools/TrackPerformanceReview/Subwidgets/FindBRDependencies/FindBRDependenciesTool
graphs [
 AddObjectActionsToList {nodes [
   AddObjectActionsToList!() -> (
    then
    InObject:Object Reference
    InOutActionEntries:Map of Strings to Find BRDependencies Tool Entry Object References
    InActions:Array of Strings
    InAssets:Array of Object References
    bObjectInLevel:bool=~
   ) {is=title K2Node_FunctionEntry}
   > n1 = For Each Loop(Array=Get InActions() -> (InActions:Array of Strings))
   > Set Action(Action:String=@n1) -> (Output_Get:String)
   > n2 = Sequence()
   n3 = Find(
    TargetMap:Map of Strings to Find BRDependencies Tool Entry Object References=Get InOutActionEntries() -> (InOutActionEntries:Map of Strings to Find BRDependencies Tool Entry Object References)
    Key=Get Action() -> (Action:String)
   ) -> (Value:Find BRDependencies Tool Entry Object Reference)
   AddObject!(
    execute=@n2.then_1
    self:Find BRDependencies Tool Entry Object Reference=@n3
    InObject:Object Reference=Get InObject() -> (InObject:Object Reference)
   ) -> (then)
   > AddActionAssets!(
    execute
    self:Find BRDependencies Tool Entry Object Reference=@n3
    InAssets:Array of Object References=Get InAssets() -> (InAssets:Array of Object References)
   ) -> (then)
   n4 = Get Action()
   n5 = Get InOutActionEntries()
   Branch(
    execute=@n2.then_0
    Condition=NOT Boolean(A=Contains!(TargetMap:Map of Strings to Find BRDependencies Tool Entry Object References=@n5, Key:"String (by ref)"=@n4) -> (ReturnValue))
   )
   > n6 = Construct Object from Class!(
    execute
    Class=&/FindBRDependenciesToolEntry.FindBRDependenciesToolEntry_C
    self
    Action:String=Get Action()
    bObjectsInLevel:bool=Get bObjectInLevel() -> (bObjectInLevel:bool)
   ) -> (then, ReturnValue:Find BRDependencies Tool Entry Object Reference)
   > Set NewEntryWidget(NewEntryWidget:Find BRDependencies Tool Entry Object Reference=@n6) -> (Output_Get:Find BRDependencies Tool Entry Object Reference)
   > Add!(
    execute
    TargetMap:Map of Strings to Find BRDependencies Tool Entry Object References=Get InOutActionEntries()
    Key:"String (by ref)"=Get Action()
    Value:"Find BRDependencies Tool Entry Object Reference (by ref)"=Get NewEntryWidget() -> (NewEntryWidget:Find BRDependencies Tool Entry Object Reference)
   ) -> (then) {is=title K2Node_CallFunction}
   > AddChild(self=Get EntriesScrollBox() -> (EntriesScrollBox:Editor Utility Scroll Box Object Reference), Content=Get NewEntryWidget())
  ]}
 ConstructActionsListForAssets {nodes [
   ConstructActionsListForAssets!() -> (then, Assets:Array of Asset Data Structures) {is=title K2Node_FunctionEntry}
   > ClearChildren(self=Get EntriesScrollBox())
   > SetVisibility(self=Get ResultsVerticalBox() -> (ResultsVerticalBox:Vertical Box Object Reference), InVisibility=SelfHitTestInvisible)
   > n1 = For Each Loop(Array:Array of Asset Data Structures=Get Assets() -> (Assets:Array of Asset Data Structures)) -> (Array Element:Asset Data Structure)
   n2 = For Each Loop(
    Exec=@n1.Completed
    Array:Array of Material Interface Object References=Get Materials() -> (Materials:Array of Material Interface Object References)
   ) -> (Array Element:Material Interface Object Reference)
   > n3 = FindActionsForObject!(execute, self, InObject:Object Reference=@n2) -> (then, OutActions:Array of Strings, OutAssets:Array of Object References)
   > AddObjectActionsToList!(
    execute
    self
    InObject:Object Reference=@n2
    InOutActionEntries:Map of Strings to Find BRDependencies Tool Entry Object References=Get ActionEntries() -> (ActionEntries:Map of Strings to Find BRDependencies Tool Entry Object References)
    InActions:Array of Strings=@n3
    InAssets:Array of Object References=@n3.OutAssets
    bObjectInLevel:bool=false
   ) -> (then) {is=title K2Node_CallFunction}
   n4 = GetAsset(InAssetData=@n1)
   n5 = Cast To MaterialInterface!(execute=@n1.LoopBody, Object=@n4) -> (then, CastFailed, AsMaterial Interface:Material Interface Object Reference) {is=title K2Node_DynamicCast}
   > Add(TargetArray:Array of Material Interface Object References=Get Materials(), NewItem:"Material Interface Object Reference (by ref)"=@n5) {is=title K2Node_CallArrayFunction}
   n6 = Cast To StaticMesh(execute=@n5.CastFailed, Object=@n4)
   > n7 = For Each Loop(Array:Array of Static Material Structures=Get StaticMaterials!(self:Static Mesh Object Reference=@n6) -> (StaticMaterials:Array of Static Material Structures)) -> (Array Element:Static Material Structure)
   > Add(
    TargetArray=Get Materials()
    NewItem=Break Static Material!(StaticMaterial:"Static Material Structure (by ref)"=@n7) -> (
     MaterialInterface:Material Interface Object Reference
     MaterialSlotName:Name
     UVChannelData:Mesh UVChannel Info Structure
     OverlayMaterialInterface:Material Interface Object Reference
    ) {is=title K2Node_BreakStruct}
   )
   n8 = Cast To SkeletalMesh!(execute=@n6.CastFailed, Object=@n4) -> (then, CastFailed, AsSkeletal Mesh:Skeletal Mesh Object Reference) {is=title K2Node_DynamicCast}
   > n9 = For Each Loop(Array:Array of Skeletal Material Structures=Get Materials!(self:Skeletal Mesh Object Reference=@n8) -> (Materials:Array of Skeletal Material Structures)) -> (Array Element:Skeletal Material Structure)
   > Add(
    TargetArray=Get Materials!() -> (Materials:Array of Material Interface Object References)
    NewItem=Break Skeletal Material!(SkeletalMaterial:"Skeletal Material Structure (by ref)"=@n9) -> (
     MaterialInterface:Material Interface Object Reference
     MaterialSlotName:Name
     UVChannelData:Mesh UVChannel Info Structure
     OverlayMaterialInterface:Material Interface Object Reference
    ) {is=title K2Node_BreakStruct}
   )
  ]}
 ConstructActionsListForLevelActors {nodes [
   ConstructActionsListForLevelActors!() -> (then, InActors:Array of Actor Object References) {is=title K2Node_FunctionEntry}
   > ClearChildren(self=Get EntriesScrollBox())
   > SetVisibility(self=Get ResultsVerticalBox(), InVisibility=SelfHitTestInvisible)
   > n1 = For Each Loop(Array:Array of Actor Object References=Get InActors() -> (InActors:Array of Actor Object References)) -> (Array Element:Actor Object Reference)
   > Set Actor(Actor:Actor Object Reference=@n1) -> (Output_Get:Actor Object Reference)
   > n2 = For Each Loop(Array:Array of Mesh Component Object References=Get Components By Class!(
     self:Actor Object Reference=Get Actor() -> (Actor:Actor Object Reference)
     ComponentClass:Actor Component Class Reference=/Script/Engine.MeshComponent
    ) -> (ReturnValue:Array of Mesh Component Object References)) -> (Array Element:Mesh Component Object Reference)
   > n3 = For Each Loop(Array:Array of Material Interface Object References=GetMaterials!(self:Mesh Component Object Reference=@n2) -> (ReturnValue:Array of Material Interface Object References)) -> (Array Element:Material Interface Object Reference)
   > n4 = FindActionsForObject(InObject=@n3)
   > AddObjectActionsToList(InObject=Get Actor(), InOutActionEntries=Get ActionEntries(), InActions=@n4, InAssets=@n4.OutAssets, bObjectInLevel=true)
  ]}
 DoesMaterialDuplicationAlreadyExist {nodes [
   DoesMaterialDuplicationAlreadyExist!() -> (then, InOriginalMaterial:Material Object Reference) {is=title K2Node_FunctionEntry}
   > n1 = ListAssets!(
    execute
    DirectoryPath:String=Get FortniteMaterialDuplicationsFolder() -> (FortniteMaterialDuplicationsFolder:String)
    bRecursive:bool=true
    bIncludeFolder:bool=false
   ) -> (then, ReturnValue:Array of Strings)
   > n2 = For Each Loop(Array:Array of Strings=@n1) -> (Array Element:String)
   Return Node!(execute=@n2.Completed, ReturnValue:bool=false)
   n3 = Get InOriginalMaterial() -> (InOriginalMaterial:Material Object Reference)
   > Branch(Condition="Equal Exactly (String)"(
     A=Split!(InPath:String=@n2) -> (PathPart:String, FilenamePart:String, ExtensionPart:String).FilenamePart
     B=GetNameForMaterialDuplication!(self, Material:Material Object Reference=@n3) -> (ReturnValue:String)
    ))
   > Return Node(ReturnValue=true)
  ]}
 EvaluateCurrentActor {nodes [
   EvaluateCurrentActor() {is=title K2Node_FunctionEntry}
   > n1 = Set Actor(Actor="Get (a copy)"(
     Array:Array of Actor Object References=Get ActorsToSearch() -> (ActorsToSearch:Array of Actor Object References)
     Dimension 1=Get CurrentIndex() -> (CurrentIndex:int=0)
    ) -> (Output:Actor Object Reference))
   > n2 = Set BuildingActor(BuildingActor:Building Level Instance Object Reference=Cast To BuildingLevelInstance!(Object=@n1) -> (AsBuilding Level Instance:Building Level Instance Object Reference, bSuccess:bool=~) {is=title K2Node_DynamicCast}) -> (Output_Get:Building Level Instance Object Reference)
   > n3 = Is Valid(InputObject=@n2)
   > n4 = GetActorsInLevel!(execute, self:Building Level Instance Object Reference=Get BuildingActor() -> (BuildingActor:Building Level Instance Object Reference)) -> (then, ReturnValue:Array of Actor Object References)
   > n5 = For Each Loop(Array:Array of Actor Object References=@n4) -> (Array Element:Actor Object Reference)
   > n6 = Set Actor(Actor=@n5)
   n7 = For Each Loop(
    Exec=[@n6.then, @n3.Is Not Valid]
    Array:Array of Mesh Component Object References=Get Components By Class(self=Get Actor(), ComponentClass=&MeshComponent)
   ) -> (Array Element:Mesh Component Object Reference)
   > n8 = For Each Loop(Array:Array of Material Interface Object References=GetMaterials(self=@n7)) -> (Array Element:Material Interface Object Reference)
   > n9 = FindActionsForObject(InObject=@n8)
   n10 = Get BuildingActor()
   > AddObjectActionsToList(
    InObject=Select(Option 0:Object Reference=Get Actor(), Option 1:Object Reference=Get BuildingActor(), Index=IsValid(Object=@n10)) -> (ReturnValue:Object Reference)
    InOutActionEntries=Get Entries() -> (Entries:Map of Strings to Find BRDependencies Tool Entry Object References)
    InActions=@n9
    InAssets=@n9.OutAssets
    bObjectInLevel=true
   )
  ]}
 EventGraph {nodes [
   "On Clicked (SearchInLevelButton)"() {is=title K2Node_ComponentBoundEvent}
   > n1 = Branch(Condition=Get bSelectedActorsOnly() -> (bSelectedActorsOnly:bool))
   > n2 = GetSelectedLevelActors(self=Get EditorActorSubsystem())
   > n3 = Set ActorsToSearch(ActorsToSearch:Array of Actor Object References=@n2) -> (Output_Get:Array of Actor Object References)
   Setup for Evaluating(execute=[@n3.then, @n5.then])
   > Evaluate Next Actor()
   n4 = GetAllLevelActors(execute=@n1.else, self=Get EditorActorSubsystem())
   > n5 = Set ActorsToSearch(ActorsToSearch=@n4)
   n6 = "On Clicked (CancelButton)"() {is=title K2Node_ComponentBoundEvent}
   Stop(execute=[@n6.then, @n10.then])
   "On Clicked (SearchInContentBrowser)"() {is=title K2Node_ComponentBoundEvent}
   > n7 = GetSelectedAssetData()
   > Set AssetsToSearch(AssetsToSearch:Array of Asset Data Structures=@n7) -> (Output_Get:Array of Asset Data Structures)
   > Setup for Evaluating()
   > Evaluate Next Asset()
   "On Clicked (ExporttoCSVButton)"() {is=title K2Node_ComponentBoundEvent}
   > Export CSV()
   Setup for Evaluating() {is=title K2Node_CustomEvent}
   > ClearChildren(self=Get EntriesScrollBox())
   > SetVisibility(self=Get ResultsVerticalBox(), InVisibility=SelfHitTestInvisible)
   > Set CurrentIndex(CurrentIndex:int=0) -> (Output_Get:int=0)
   > Set bShouldStop(bShouldStop:bool=false) -> (Output_Get:bool)
   > Clear!(execute, TargetMap:Map of Strings to Find BRDependencies Tool Entry Object References=Get Entries()) -> (then) {is=title K2Node_CallFunction}
   > SetVisibility(
    self=[
     Get CancelButton() -> (CancelButton:Editor Utility Button Object Reference)
     Get ProgressCounterText() -> (ProgressCounterText:Editor Utility Editable Text Object Reference)
    ]
    InVisibility=Visible
   )
   Stop() {is=title K2Node_CustomEvent}
   > Set bShouldStop(bShouldStop=true)
   > SetPercent!(
    execute
    self:Progress Bar Object Reference=Get AsyncProgressBar() -> (AsyncProgressBar:Editor Utility Progress Bar Object Reference)
    InPercent:float=0
   ) -> (then)
   > SetVisibility(self=[Get CancelButton(), Get ProgressCounterText()], InVisibility=Collapsed)
   Event PreConstruct()
   > SetObject(self=Get ParametersView() -> (ParametersView:Details View Object Reference), NewObject=Self-Reference())
   > Do Once()
   > n8 = GetCurrentLevel!(execute, self:Level Editor Subsystem Object Reference=Get LevelEditorSubsystem()) -> (then, ReturnValue:Level Object Reference)
   n9 = GetDisplayName(Object=GetOuterObject!(Object:Object Reference=@n8) -> (ReturnValue:Object Reference))
   > Set Export CSV Path(Export CSV Path:String=Append(A=GetProjectSavedDirectory!() -> (ReturnValue:String), B=Append(A=@n9, B=_BR_Dependencies.csv))) -> (Output_Get:String)
   n10 = Event Destruct()
   K2Node_Composite_0 = Evaluate Level Actors Graph!() {is=title K2Node_Composite}
   K2Node_Composite_1 = Evaluate Assets Graph!() {is=title K2Node_Composite}
  ]}
 Evaluate Level Actors Graph {
  path=EventGraph.K2Node_Composite_0.Evaluate Level Actors Graph
  nodes [
   Evaluate Next Actor() {is=title K2Node_CustomEvent}
   n1 = Get ActorsToSearch()
   > n2 = Branch(Condition="integer < integer"(A=Get CurrentIndex() -> (CurrentIndex=0), B=Length(TargetArray=@n1)))
   > UpdateCounterText!(execute, self, Total:int=Length(TargetArray=Get ActorsToSearch())) -> (then)
   > n3 = Sequence()
   > EvaluateCurrentActor() {is=title K2Node_CallFunction}
   Increment Int(" "=@n3.then_1, Value=Get CurrentIndex() -> (CurrentIndex=0))
   n4 = Length(TargetArray=Get ActorsToSearch())
   n5 = "To Float (Integer)"(InInt=Get CurrentIndex() -> (CurrentIndex=0))
   > SetPercent(self=Get AsyncProgressBar(), InPercent=float / float(A=float + float(A=@n5, B=1), B="To Float (Integer)"(InInt=@n4)))
   > n6 = AsyncEditorDelay!(execute:Exec, Seconds:float=0.008, MinimumFrames:int=0) -> (then:Exec, Complete:Exec) {is=title K2Node_AsyncAction}
   n7 = Branch(execute=@n6.Complete, Condition=Get bShouldStop() -> (bShouldStop:bool))
   Evaluate Next Actor(execute=@n7.else) {is=title K2Node_CallFunction}
   Stop(execute=@n2.else) {is=title K2Node_CallFunction}
   Inputs!()
   Outputs()
  ]
 }
 Evaluate Assets Graph {
  path=EventGraph.K2Node_Composite_1.Evaluate Assets Graph
  nodes [
   Evaluate Next Asset() {is=title K2Node_CustomEvent}
   n1 = Get AssetsToSearch() -> (AssetsToSearch:Array of Asset Data Structures)
   > n2 = Branch(Condition="integer < integer"(A=Get CurrentIndex() -> (CurrentIndex=0), B=Length(TargetArray:Array of Asset Data Structures=@n1)))
   > UpdateCounterText(Total=Length(TargetArray=Get AssetsToSearch()))
   n3 = GetAsset(InAssetData="Get (a copy)"(Array:Array of Asset Data Structures=Get AssetsToSearch(), Dimension 1=Get CurrentIndex() -> (CurrentIndex=0)) -> (Output:Asset Data Structure))
   > n4 = GetAssetMaterials!(execute, self, InAsset:Object Reference=@n3) -> (then, ReturnValue:Array of Material Interface Object References)
   > n5 = For Each Loop(Array=@n4)
   Increment Int(" "=@n5.Completed, Value=Get CurrentIndex() -> (CurrentIndex=0))
   n6 = Length(TargetArray=Get AssetsToSearch())
   n7 = "To Float (Integer)"(InInt=Get CurrentIndex() -> (CurrentIndex=0))
   > SetPercent(self=Get AsyncProgressBar(), InPercent=float / float(A=float + float(A=@n7, B=1), B="To Float (Integer)"(InInt=@n6)))
   > n8 = AsyncEditorDelay(Seconds=0.008, MinimumFrames=0)
   n9 = Branch(execute=@n8.Complete, Condition=Get bShouldStop())
   Evaluate Next Asset(execute=@n9.else) {is=title K2Node_CallFunction}
   n10 = FindActionsForObject(execute=@n5.LoopBody, InObject=@n5)
   > AddObjectActionsToList(InObject=@n3, InOutActionEntries=Get Entries(), InActions=@n10, InAssets=@n10.OutAssets, bObjectInLevel=false)
   Stop(execute=@n2.else)
   Inputs()
   Outputs()
  ]
 }
 Export CSV {nodes [
   Export CSV() {is=title K2Node_FunctionEntry}
   > n1 = Sequence()
   > Set CSV Content(CSV Content:String="Affected Objects, Suggested Actions") -> (Output_Get:String)
   n2 = Get EntriesScrollBox()
   > n3 = For Loop(LastIndex=int - int(A=GetChildrenCount!(self:Panel Widget Object Reference=@n2) -> (ReturnValue:int=0)))
   > n4 = Cast To FindBRDependenciesToolEntry!(
    execute
    Object=GetChildAt!(self:Panel Widget Object Reference=Get EntriesScrollBox(), Index:int=@n3) -> (ReturnValue:Widget Object Reference)
   ) -> (then, CastFailed, AsFind BRDependencies Tool Entry:Find BRDependencies Tool Entry Object Reference) {is=title K2Node_DynamicCast}
   n5 = Get ActionText!(self:Find BRDependencies Tool Entry Object Reference=@n4) -> (ActionText:Editor Utility Editable Text Object Reference)
   n6 = Get ActorsText!(self:Find BRDependencies Tool Entry Object Reference=@n4) -> (ActorsText:Editor Utility Editable Text Object Reference)
   n7 = Format Text!(
    Format="INVTEXT(\"\\r\\n\\\"{Actors}\\\",\\\"{Action}\\\"\")"
    Actors:String="To String (Text)"(InText="GetText (Editable Text)"!(self:Editable Text Object Reference=@n6) -> (ReturnValue:Text))
    Action:String="To String (Text)"(InText="GetText (Editable Text)"(self=@n5))
   ) -> (Result)
   > Set CSV Content(CSV Content=Append(A=Get CSV Content() -> (CSV Content:String), B="To String (Text)"(InText=@n7)))
   n8 = Get Export CSV Path() -> (Export CSV Path:String)
   n9 = Get CSV Content()
   n10 = "Execute Python Command (Advanced)"!(
    execute=@n1.then_1
    PythonCommand:String="To String (Text)"(InText=Format Text!(
      Format="INVTEXT(\"f=open(\\\"{File Name}\\\", \\\"w\\\")\\r\\nf.write(\\\"\\\"\\\"{content}\\r\\n\\\"\\\"\\\")\\r\\nf.close()\")"
      File Name:String=@n8
      content:String=@n9
     ) -> (Result))
    ExecutionMode:EPythonCommandExecutionMode Enum=ExecuteFile
    FileExecutionScope:EPythonFileExecutionScope Enum=Private
   ) -> (then, CommandResult:String, LogOutput:Array of Python Log Output Entry Structures, ReturnValue:bool)
   > Show Message Dialog(
    Title=MakeLiteralText(Value="INVTEXT(\"CSV Export\")")
    Message=Select(
     Option 0:"Text (by ref)"="To Text (String)"(InString=@n10)
     Option 1:"Text (by ref)"=Format Text!(Format="INVTEXT(\"Successfuly exported to CSV\\r\\n{File Name}\")", File Name=@n8) -> (Result)
     Index=@n10.ReturnValue
    ) -> (ReturnValue:"Text (by ref)")
    MessageType=Ok
    DefaultValue=No
    MessageCategory=Select(Option 0:EAppMsgCategory Enum=Error, Option 1:EAppMsgCategory Enum=Success, Index=@n10.ReturnValue) -> (ReturnValue:EAppMsgCategory Enum)
   )
  ]}
 FindActionsForObject {nodes [
   FindActionsForObject!() -> (then, InObject:Object Reference) {is=title K2Node_FunctionEntry}
   > FindProblematicDependencies!(
    execute
    self
    InObject:Object Reference=Get InObject()
    InOutDependencies:Array of Object References=Get ProblematicDependenciesFound() -> (ProblematicDependenciesFound:Array of Object References)
   ) -> (then)
   > n1 = GetActionsForProblematicDependencies!(execute, self, InObject:Object Reference=Get InObject(), InDependencies:Array of Object References=Get ProblematicDependenciesFound()) -> (then, OutActions:Array of Strings, OutAssets:Array of Object References)
   > Return Node!(execute, OutActions:Array of Strings=@n1, OutAssets:Array of Object References=@n1.OutAssets)
  ]}
 FindProblematicDependencies {nodes [
   FindProblematicDependencies!() -> (then, InObject:Object Reference, InOutDependencies:Array of Object References) {is=title K2Node_FunctionEntry}
   > n1 = Sequence()
   > n2 = Cast To MaterialInstanceDynamic!(execute, Object=Get InObject()) -> (then, CastFailed, AsMaterial Instance Dynamic:Material Instance Dynamic Object Reference) {is=title K2Node_DynamicCast}
   > FindProblematicDependencies(
    InObject=Get Parent!(self:Material Instance Object Reference=@n2) -> (Parent:Material Interface Object Reference)
    InOutDependencies=Get InOutDependencies() -> (InOutDependencies:Array of Object References)
   ) {is=title K2Node_CallFunction}
   n3 = Get Object Path String(Object=Get InObject())
   n4 = Get Dependencies!(
    execute=@n1.then_1
    self:Asset Registry Interface=GetAssetRegistry()
    PackageName:Name=String To Name(InString=Split!(SourceString:String=@n3, InStr:String=., SearchCase:ESearchCase Enum, SearchDir:ESearchDir Enum=FromStart) -> (LeftS:String, RightS:String, ReturnValue:bool))
    DependencyOptions:"Asset Registry Dependency Options Structure (by ref)"=Get DependencyOptions() -> (DependencyOptions:Asset Registry Dependency Options Structure)
   ) -> (then, OutDependencies:Array of Names, ReturnValue:bool) {is=title K2Node_CallFunction}
   > Set DependencyPaths(DependencyPaths:Array of Names=@n4) -> (Output_Get:Array of Names)
   > Branch(Condition=@n4.ReturnValue)
   > n5 = For Loop(LastIndex=Last Index!(TargetArray:Array of Names=Get DependencyPaths() -> (DependencyPaths:Array of Names)) -> (ReturnValue=0))
   n6 = Get DependencyPaths()
   > n7 = LoadAsset!(
    execute
    self:Editor Asset Subsystem Object Reference=Get EditorAssetSubsystem() -> (ReturnValue:Editor Asset Subsystem Object Reference) {is=title K2Node_GetEditorSubsystem}
    AssetPath="To String (Name)"(InName="Get (a ref)"(Array:Array of Names=@n6, Dimension 1=@n5) -> (Output:"Name (by ref)"))
   ) -> (then, ReturnValue)
   > n8 = Set DependencyObject(DependencyObject:Object Reference=@n7) -> (Output_Get:Object Reference)
   > Is Valid(InputObject=@n8)
   > n9 = Branch(Condition=Contains Item(
     TargetArray:Array of Object References=Get ProblematicDependencies() -> (ProblematicDependencies:Array of Object References)
     ItemToFind:"Object Reference (by ref)"=Get DependencyObject() -> (DependencyObject:Object Reference)
    ))
   > Add Unique(TargetArray:Array of Object References=Get InOutDependencies(), NewItem:"Object Reference (by ref)"=Get DependencyObject())
   Branch(execute=@n9.else, Condition=IsObjectRelevant!(self, InObject:Object Reference=Get DependencyObject()) -> (ReturnValue:bool))
   > FindProblematicDependencies(InObject=Get DependencyObject(), InOutDependencies=Get InOutDependencies())
  ]}
 GetActionsForProblematicDependencies {nodes [
   GetActionsForProblematicDependencies!() -> (then, InObject:Object Reference, InDependencies:Array of Object References) {is=title K2Node_FunctionEntry}
   n1 = Get InDependencies() -> (InDependencies:Array of Object References)
   > n2 = Branch(Condition="integer > integer"(A=Length(TargetArray:Array of Object References=@n1)))
   > n3 = Sequence()
   > n4 = Cast To MaterialInterface(Object=Get InObject())
   > Set ParentMaterial(ParentMaterial:Material Object Reference=GetParentMaterial!(self, InMaterial:Material Interface Object Reference=@n4) -> (ReturnValue:Material Object Reference)) -> (Output_Get:Material Object Reference)
   > n5 = Branch(Condition=DoesMaterialDuplicationAlreadyExist!(self, InOriginalMaterial:Material Object Reference=Get ParentMaterial() -> (ParentMaterial:Material Object Reference)) -> (ReturnValue:bool) {is=title K2Node_CallFunction})
   n6 = GetNameForMaterialDuplication(Material=Get ParentMaterial())
   n7 = GetDisplayName(Object=Get ParentMaterial())
   n8 = Get FortniteMaterialDuplicationsFolder()
   > Add(
    TargetArray:Array of Strings=Get Actions() -> (Actions:Array of Strings)
    NewItem:"String (by ref)"="To String (Text)"(InText=Format Text!(
      Format="INVTEXT(\"Substitute all usages of {Material} by {Duplication} in this level.\\r\\nYou can find it in {Path}\")"
      Material:String=@n7
      Duplication:String=@n6
      Path:String=@n8
     ) -> (Result))
   )
   > n9 = LoadMaterialDuplication!(execute, self, Material:Material Object Reference=Get ParentMaterial()) -> (then, ReturnValue:Object Reference)
   > Add(TargetArray:Array of Object References=Get Assets() -> (Assets:Array of Object References), NewItem:"Object Reference (by ref)"=@n9)
   n10 = Sequence(execute=@n5.else)
   > n11 = For Each Loop(Array:Array of Object References=Get InDependencies()) -> (Array Element:Object Reference)
   > n12 = Cast To MaterialFunction!(execute, Object=@n11) -> (then, CastFailed, AsMaterial Function:Material Function Object Reference) {is=title K2Node_DynamicCast}
   n13 = Last Index(TargetArray:Array of Object References=Get InDependencies()) -> (ReturnValue=0)
   n14 = "Equal (Integer)"(A=@n11.Array Index, B=@n13)
   n15 = Select(Option 0:String=", ", Option 1:String=" and ", Index="Equal (Integer)"(A=@n11.Array Index, B=int - int(A=@n13))) -> (ReturnValue:String)
   > Set ReferencedMaterialFunctionsNames(ReferencedMaterialFunctionsNames:String=Append!(
     A=Get ReferencedMaterialFunctionsNames() -> (ReferencedMaterialFunctionsNames:String)
     B=GetDisplayName(Object=@n12)
     C:String=Select(Option 0=@n15, Index=@n14)
    ) -> (ReturnValue)) -> (Output_Get:String)
   n16 = GetNameForMaterialDuplication(Material=Get ParentMaterial())
   n17 = GetDisplayName(Object=Get ParentMaterial())
   n18 = Get FortniteMaterialDuplicationsFolder()
   n19 = Get ReferencedMaterialFunctionsNames()
   Add(
    execute=@n10.then_1
    TargetArray:Array of Strings=Get Actions()
    NewItem:"String (by ref)"="To String (Text)"(InText=Format Text!(
      Format="INVTEXT(\"1. Create a duplication of {Material} and name it {Duplication}.\\r\\n2. Place it in {Path}\\r\\n3. Make sure it doesn\\'t use {MaterialFunctions}.\\r\\n4. Create new Material Instances that use the duplicated material.\\r\\n5. Override the usages of {Material} in the level.\")"
      Material=@n17
      Duplication=@n16
      Path=@n18
      MaterialFunctions:String=@n19
     ) -> (Result))
   )
   > Add(TargetArray:Array of Object References=Get Assets(), NewItem:"Object Reference (by ref)"=Get ParentMaterial())
   Return Node(execute=[@n3.then_1, @n2.else], OutActions=Get Actions(), OutAssets=Get Assets())
  ]}
 GetAssetMaterials {nodes [
   n1 = GetAssetMaterials!() -> (then, InAsset:Object Reference) {is=title K2Node_FunctionEntry}
   > n2 = Sequence()
   > n3 = Cast To MaterialInterface(Object=@n1)
   > Set Materials(Materials:Array of Material Interface Object References=Make Array("[0]":Material Interface Object Reference=@n3) -> (Array:Array of Material Interface Object References)) -> (Output_Get:Array of Material Interface Object References)
   n4 = Cast To StaticMesh(execute=@n3.CastFailed, Object=@n1)
   > n5 = For Each Loop(Array:Array of Static Material Structures=Get StaticMaterials(self=@n4)) -> (Array Element:Static Material Structure)
   > Add Unique(
    TargetArray:Array of Material Interface Object References=Get Materials()
    NewItem:"Material Interface Object Reference (by ref)"=Break Static Material(StaticMaterial=@n5)
   )
   n6 = Cast To SkeletalMesh(execute=@n4.CastFailed, Object=@n1)
   > n7 = For Each Loop(Array:Array of Skeletal Material Structures=Get Materials!(self:Skeletal Mesh Object Reference=@n6) -> (Materials:Array of Skeletal Material Structures)) -> (Array Element:Skeletal Material Structure)
   > Add Unique(
    TargetArray=Get Materials!() -> (Materials:Array of Material Interface Object References)
    NewItem=Break Skeletal Material(SkeletalMaterial=@n7)
   )
   Return Node!(execute=@n2.then_1, ReturnValue:Array of Material Interface Object References=Get Materials())
  ]}
 GetNameForMaterialDuplication {nodes [
   GetNameForMaterialDuplication!() -> (then, Material:Material Object Reference) {is=title K2Node_FunctionEntry}
   n1 = GetDisplayName(Object=Get Material() -> (Material:Material Object Reference))
   n2 = Len(S=@n1)
   n3 = "Equal Exactly (String)"(A=LeftChop(SourceString=@n1, Count=int - int(A=@n2, B=2)), B=M_)
   n4 = GetDisplayName(Object=Get Material())
   n5 = RightChop!(SourceString:String=@n1, Count:int=2) -> (ReturnValue:String)
   > Return Node(ReturnValue:String=Append!(A=Get FortniteMaterialDuplicationsPrefix() -> (FortniteMaterialDuplicationsPrefix:String), B=Select(Option 0=@n4, Option 1=@n5, Index=@n3)) -> (ReturnValue))
  ]}
 GetParentMaterial {nodes [
   n1 = GetParentMaterial!() -> (then, InMaterial:Material Interface Object Reference) {is=title K2Node_FunctionEntry}
   > n2 = Cast To Material!(execute, Object=@n1) -> (then, CastFailed, AsMaterial:Material Object Reference) {is=title K2Node_DynamicCast}
   > Return Node(ReturnValue:Material Object Reference=@n2)
   n3 = Cast To MaterialInstance!(execute=@n2.CastFailed, Object=@n1) -> (then, CastFailed, AsMaterial Instance:Material Instance Object Reference) {is=title K2Node_DynamicCast}
   > Return Node(ReturnValue=GetParentMaterial(InMaterial=Get Parent(self=@n3)) {is=title K2Node_CallFunction})
  ]}
 IsObjectRelevant {nodes [
   IsObjectRelevant!() -> (then, InObject:Object Reference) {is=title K2Node_FunctionEntry}
   > n1 = For Each Loop(Array:Array of Object Class References=Get RelevantObjectTypes() -> (RelevantObjectTypes:Array of Object Class References)) -> (Array Element:Object Class Reference)
   Return Node(execute=@n1.Completed, ReturnValue:bool=false)
   > Branch(Condition=IsA!(Object:Object Reference=Get InObject(), ObjectClass:Object Class Reference=@n1) -> (ReturnValue:bool))
   > Return Node(ReturnValue=true)
  ]}
 LoadMaterialDuplication {nodes [
   n1 = LoadMaterialDuplication!() -> (then, Material:Material Object Reference) {is=title K2Node_FunctionEntry}
   n2 = GetNameForMaterialDuplication(Material=@n1) {is=title K2Node_CallFunction}
   > n3 = LoadAsset!(execute, AssetPath=Append!(A=Get FortniteMaterialDuplicationsFolder(), B=@n2, C:String=., D:String=@n2) -> (ReturnValue)) -> (then, ReturnValue)
   > Return Node(ReturnValue:Object Reference=@n3)
  ]}
 UpdateCounterText {nodes [
   n1 = UpdateCounterText!() -> (then, Total:int) {is=title K2Node_FunctionEntry}
   > Set Text(
    Text=Format Text!(Format="INVTEXT(\"{Current}/{Total}\")", Current:int=Get CurrentIndex() -> (CurrentIndex=0), Total:int=@n1) -> (Result)
    self:Editable Text Object Reference=Get ProgressCounterText()
   )
  ]}
]
variables [
 ActorsToSearch {type="array<Actor>"}
 AssetsToSearch {type="array<AssetData>", default []}
 bSelectedActorsOnly {type=boolean}
 bShouldStop {type=boolean}
 CurrentIndex {type=integer}
 Entries {type="map<FindBRDependenciesToolEntry_C>", default {}}
 Export CSV Path {type=string, default=""}
 FortniteMaterialDuplicationsFolder {type=string, default=/DelMarGame/Environments/Shared/Fortnite_Ported/}
 FortniteMaterialDuplicationsPrefix {type=string, default=M_DM_}
 ProblematicDependencies {
  type="array<Object>"
  default [
   &MF_FN_DestructibleDamage
   /Game/Packages/Fortress_SharedMaterials/Base_Material_Functions/MF_CurieFallbackEffect
   /Game/Packages/Fortress_SharedMaterials/Base_Material_Functions/MF_Curie_Landscape
  ]
 }
 RelevantObjectTypes {
  type="array<Class>"
  default [
   /Script/Engine.Material
   /Script/Engine.MaterialInstanceConstant
   /Script/Engine.MaterialInstanceDynamic
   /Script/Engine.MaterialFunction
   &StaticMesh
   &SkeletalMesh
  ]
 }
]
widgets {
 info {rootWidgetClass=&VerticalBox, widgetCount=27}
 widgets [
  RootVerticalBox {widgetClassPath=&VerticalBox, widget=RootVerticalBox}
  TopHorizontalBox {widgetClassPath=&HorizontalBox, parent=RootVerticalBox, widget=TopHorizontalBox, slot {class=VerticalBoxSlot}}
  ParametersView {
   widgetClassPath=&DetailsView
   parent=TopHorizontalBox
   widget=ParametersView
   properties {categoriesToShow [Parameters]}
   slot {class=HorizontalBoxSlot, properties {size {sizeRule=Fill, value=1}}}
  }
  ButtonsHorizontalBox {
   widgetClassPath=&HorizontalBox
   parent=RootVerticalBox
   widget=ButtonsHorizontalBox
   slot {class=VerticalBoxSlot, properties {verticalAlignment=VAlign_Center}}
  }
  SearchInLevelButton {
   widgetClassPath=&EditorUtilityButton
   parent=ButtonsHorizontalBox
   widget=SearchInLevelButton
   slot {class=HorizontalBoxSlot, properties {verticalAlignment=VAlign_Center, size {sizeRule=Fill, value=1}}}
  }
  EditorUtilityEditableText_1 {
   widgetClassPath=/Script/Blutility.EditorUtilityEditableText
   parent=SearchInLevelButton
   widget=EditorUtilityEditableText_1
   properties {
    isReadOnly=true
    justification=Center
    text=Search Dependencies in Level Materials
    visibility=SelfHitTestInvisible
    widgetStyle {
     backgroundImageComposing {
      bIsDynamicallyLoaded=false
      drawAs=Border
      imageType=FullColor
      mirroring=NoMirror
      resourceName=../../../Engine/Content/Slate/Old/HyperlinkDotted.png
      resourceObject=null
      tiling=Both
      imageSize=<0 0>
      margin=<ltrb 0 0 0 0.1875>
      outlineSettings {
       bUseBrushTransparency=false
       roundingType=HalfHeightRadius
       width=0
       color {colorUseRule=UseColor_Specified, specifiedColor=<rgba 0 0 0 0>}
       cornerRadii=<0 0 0 0>
      }
      tintColor {colorUseRule=UseColor_Specified, specifiedColor=<rgba 1 1 1 1>}
      uVRegion {bIsValid=false, max=<0 0>, min=<0 0>}
     }
     backgroundImageSelected {
      bIsDynamicallyLoaded=false
      drawAs=Image
      imageType=NoImage
      mirroring=NoMirror
      resourceName=None
      resourceObject=null
      tiling=NoTile
      imageSize=<0 0>
      margin=<ltrb 0 0 0 0>
      outlineSettings {
       bUseBrushTransparency=false
       roundingType=HalfHeightRadius
       width=0
       color {colorUseRule=UseColor_Specified, specifiedColor=<rgba 0 0 0 0>}
       cornerRadii=<0 0 0 0>
      }
      tintColor {colorUseRule=UseColor_Specified, specifiedColor=<rgba 0 0.162 0.7454 1>}
      uVRegion {bIsValid=false, max=<0 0>, min=<0 0>}
     }
     caretImage {
      bIsDynamicallyLoaded=false
      drawAs=Image
      imageType=NoImage
      mirroring=NoMirror
      resourceName=None
      resourceObject=null
      tiling=NoTile
      imageSize=<0 0>
      margin=<ltrb 0 0 0 0>
      outlineSettings {
       bUseBrushTransparency=false
       roundingType=HalfHeightRadius
       width=0
       color {colorUseRule=UseColor_Specified, specifiedColor=<rgba 0 0 0 0>}
       cornerRadii=<0 0 0 0>
      }
      tintColor {colorUseRule=UseColor_Specified, specifiedColor=<rgba 1 1 1 1>}
      uVRegion {bIsValid=false, max=<0 0>, min=<0 0>}
     }
     colorAndOpacity {colorUseRule=UseColor_Specified, specifiedColor=<rgba 0.5271 0.5271 0.5271 1>}
     font {
      bForceMonospaced=false
      bMaterialIsStencil=false
      fontMaterial=null
      fontObject=null
      letterSpacing=0
      monospacedWidth=1
      size=9.75
      skewAmount=0
      typefaceFontName=Regular
      outlineSettings {
       bApplyOutlineToDropShadows=false
       bMiteredCorners=false
       bSeparateFillAlpha=false
       outlineMaterial=null
       outlineSize=0
       outlineColor=<rgba 0 0 0 1>
      }
     }
     hintColorAndOpacity {colorUseRule=UseColor_Foreground, specifiedColor=<rgba 1 0 1 1>}
     selectedBackgroundColor {colorUseRule=UseColor_Foreground, specifiedColor=<rgba 1 0 1 1>}
    }
   }
   slot {class=ButtonSlot, properties {horizontalAlignment=HAlign_Fill, verticalAlignment=VAlign_Fill, padding=<ltrb 5 5 5 5>}}
  }
  SearchInContentBrowser {
   widgetClassPath=&EditorUtilityButton
   parent=ButtonsHorizontalBox
   widget=SearchInContentBrowser
   slot {class=HorizontalBoxSlot, properties {verticalAlignment=VAlign_Center, size {sizeRule=Fill, value=1}}}
  }
  EditorUtilityEditableText {
   widgetClassPath=&EditorUtilityEditableText
   parent=SearchInContentBrowser
   widget=EditorUtilityEditableText
   properties {
    isReadOnly=true
    justification=Center
    text=Search Dependencies in Content Browser Selection
    visibility=SelfHitTestInvisible
    widgetStyle {
     backgroundImageComposing {
      bIsDynamicallyLoaded=false
      drawAs=Border
      imageType=FullColor
      mirroring=NoMirror
      resourceName=../../../Engine/Content/Slate/Old/HyperlinkDotted.png
      resourceObject=null
      tiling=Both
      imageSize=<0 0>
      margin=<ltrb 0 0 0 0.1875>
      outlineSettings {
       bUseBrushTransparency=false
       roundingType=HalfHeightRadius
       width=0
       color {colorUseRule=UseColor_Specified, specifiedColor=<rgba 0 0 0 0>}
       cornerRadii=<0 0 0 0>
      }
      tintColor {colorUseRule=UseColor_Specified, specifiedColor=<rgba 1 1 1 1>}
      uVRegion {bIsValid=false, max=<0 0>, min=<0 0>}
     }
     backgroundImageSelected {
      bIsDynamicallyLoaded=false
      drawAs=Image
      imageType=NoImage
      mirroring=NoMirror
      resourceName=None
      resourceObject=null
      tiling=NoTile
      imageSize=<0 0>
      margin=<ltrb 0 0 0 0>
      outlineSettings {
       bUseBrushTransparency=false
       roundingType=HalfHeightRadius
       width=0
       color {colorUseRule=UseColor_Specified, specifiedColor=<rgba 0 0 0 0>}
       cornerRadii=<0 0 0 0>
      }
      tintColor {colorUseRule=UseColor_Specified, specifiedColor=<rgba 0 0.162 0.7454 1>}
      uVRegion {bIsValid=false, max=<0 0>, min=<0 0>}
     }
     caretImage {
      bIsDynamicallyLoaded=false
      drawAs=Image
      imageType=NoImage
      mirroring=NoMirror
      resourceName=None
      resourceObject=null
      tiling=NoTile
      imageSize=<0 0>
      margin=<ltrb 0 0 0 0>
      outlineSettings {
       bUseBrushTransparency=false
       roundingType=HalfHeightRadius
       width=0
       color {colorUseRule=UseColor_Specified, specifiedColor=<rgba 0 0 0 0>}
       cornerRadii=<0 0 0 0>
      }
      tintColor {colorUseRule=UseColor_Specified, specifiedColor=<rgba 1 1 1 1>}
      uVRegion {bIsValid=false, max=<0 0>, min=<0 0>}
     }
     colorAndOpacity {colorUseRule=UseColor_Specified, specifiedColor=<rgba 0.5271 0.5271 0.5271 1>}
     font {
      bForceMonospaced=false
      bMaterialIsStencil=false
      fontMaterial=null
      fontObject=null
      letterSpacing=0
      monospacedWidth=1
      size=9.75
      skewAmount=0
      typefaceFontName=Regular
      outlineSettings {
       bApplyOutlineToDropShadows=false
       bMiteredCorners=false
       bSeparateFillAlpha=false
       outlineMaterial=null
       outlineSize=0
       outlineColor=<rgba 0 0 0 1>
      }
     }
     hintColorAndOpacity {colorUseRule=UseColor_Foreground, specifiedColor=<rgba 1 0 1 1>}
     selectedBackgroundColor {colorUseRule=UseColor_Foreground, specifiedColor=<rgba 1 0 1 1>}
    }
   }
   slot {class=ButtonSlot, properties {horizontalAlignment=HAlign_Fill, verticalAlignment=VAlign_Fill, padding=<ltrb 5 5 5 5>}}
  }
  HorizontalBox_93 {widgetClassPath=&HorizontalBox, parent=RootVerticalBox, widget=HorizontalBox_93, slot {class=VerticalBoxSlot}}
  CancelButton {
   widgetClassPath=&EditorUtilityButton
   parent=HorizontalBox_93
   widget=CancelButton
   properties {visibility=Collapsed}
   slot {class=HorizontalBoxSlot}
  }
  Image_152 {
   widgetClassPath=&Image
   parent=CancelButton
   widget=Image_152
   properties {brush {
     bIsDynamicallyLoaded=false
     drawAs=Image
     imageType=FullColor
     mirroring=NoMirror
     resourceName=None
     resourceObject=null
     tiling=NoTile
     imageSize=<16 16>
     margin=<ltrb 0 0 0 0>
     outlineSettings {
      bUseBrushTransparency=false
      roundingType=HalfHeightRadius
      width=0
      color {colorUseRule=UseColor_Specified, specifiedColor=<rgba 0 0 0 0>}
      cornerRadii=<0 0 0 0>
     }
     tintColor {colorUseRule=UseColor_Specified, specifiedColor=<rgba 1 1 1 1>}
     uVRegion {bIsValid=false, max=<0 0>, min=<0 0>}
    }}
   slot {class=ButtonSlot}
  }
  Overlay_96 {
   widgetClassPath=&Overlay
   parent=HorizontalBox_93
   widget=Overlay_96
   slot {class=HorizontalBoxSlot, properties {size {sizeRule=Fill, value=1}}}
  }
  AsyncProgressBar {
   widgetClassPath=/Script/Blutility.EditorUtilityProgressBar
   bIsVariable=true
   parent=Overlay_96
   widget=AsyncProgressBar
   properties {fillColorAndOpacity=<rgba 0 0.5 1 1>}
   slot {class=OverlaySlot, properties {horizontalAlignment=HAlign_Fill, verticalAlignment=VAlign_Fill}}
  }
  ProgressCounterText {
   widgetClassPath=&EditorUtilityEditableText
   bIsVariable=true
   parent=Overlay_96
   widget=ProgressCounterText
   properties {isReadOnly=true, text=0/0, visibility=Collapsed}
   slot {class=OverlaySlot, properties {horizontalAlignment=HAlign_Center, verticalAlignment=VAlign_Center}}
  }
  Divider {
   widgetClassPath=&Image
   parent=RootVerticalBox
   widget=Divider
   properties {
    brush {
     bIsDynamicallyLoaded=false
     drawAs=Image
     imageType=NoImage
     mirroring=NoMirror
     resourceName=None
     resourceObject=null
     tiling=NoTile
     imageSize=<32 5>
     margin=<ltrb 0 0 0 0>
     outlineSettings {
      bUseBrushTransparency=false
      roundingType=HalfHeightRadius
      width=0
      color {colorUseRule=UseColor_Specified, specifiedColor=<rgba 0 0 0 0>}
      cornerRadii=<0 0 0 0>
     }
     tintColor {colorUseRule=UseColor_Specified, specifiedColor=<rgba 1 1 1 1>}
     uVRegion {bIsValid=false, max=<0 0>, min=<0 0>}
    }
    colorAndOpacity=<rgba 0.0284 0.0284 0.0284 1>
   }
   slot {class=VerticalBoxSlot}
  }
  ResultsVerticalBox {
   widgetClassPath=&VerticalBox
   bIsVariable=true
   parent=RootVerticalBox
   widget=ResultsVerticalBox
   properties {visibility=Collapsed}
   slot {class=VerticalBoxSlot, properties {size {sizeRule=Fill, value=1}}}
  }
  TextBlock_3 {
   widgetClassPath=&TextBlock
   parent=ResultsVerticalBox
   widget=TextBlock_3
   properties {
    justification=Center
    text=Results
    colorAndOpacity {colorUseRule=UseColor_Specified, specifiedColor=<rgba 0.5271 0.5271 0.5271 1>}
    font {
     bForceMonospaced=false
     bMaterialIsStencil=false
     fontMaterial=null
     fontObject=&Roboto
     letterSpacing=0
     monospacedWidth=1
     size=15
     skewAmount=0
     typefaceFontName=Bold
     outlineSettings {
      bApplyOutlineToDropShadows=false
      bMiteredCorners=false
      bSeparateFillAlpha=false
      outlineMaterial=null
      outlineSize=0
      outlineColor=<rgba 0 0 0 1>
     }
    }
   }
   slot {class=VerticalBoxSlot, properties {verticalAlignment=VAlign_Center, padding=<ltrb 4 2 4 2>}}
  }
  ColumnLabelsHorizontalBox {widgetClassPath=&HorizontalBox, parent=ResultsVerticalBox, widget=ColumnLabelsHorizontalBox, slot {class=VerticalBoxSlot}}
  AffectedObjectsLabelOverlay {
   widgetClassPath=&Overlay
   parent=ColumnLabelsHorizontalBox
   widget=AffectedObjectsLabelOverlay
   slot {class=HorizontalBoxSlot, properties {padding=<ltrb 1 1 1 1>, size {sizeRule=Fill, value=1}}}
  }
  Image_109 {
   widgetClassPath=&Image
   parent=AffectedObjectsLabelOverlay
   widget=Image_109
   properties {colorAndOpacity=<rgba 0.0284 0.0284 0.0284 1>}
   slot {class=OverlaySlot, properties {horizontalAlignment=HAlign_Fill, verticalAlignment=VAlign_Fill}}
  }
  TextBlock_1 {
   widgetClassPath=&TextBlock
   parent=AffectedObjectsLabelOverlay
   widget=TextBlock_1
   properties {
    text=Affected Objects
    colorAndOpacity {colorUseRule=UseColor_Specified, specifiedColor=<rgba 0.5271 0.5271 0.5271 1>}
    font {
     bForceMonospaced=false
     bMaterialIsStencil=false
     fontMaterial=null
     fontObject=&Roboto
     letterSpacing=0
     monospacedWidth=1
     size=9
     skewAmount=0
     typefaceFontName=Bold
     outlineSettings {
      bApplyOutlineToDropShadows=false
      bMiteredCorners=false
      bSeparateFillAlpha=false
      outlineMaterial=null
      outlineSize=0
      outlineColor=<rgba 0 0 0 1>
     }
    }
   }
   slot {class=OverlaySlot, properties {horizontalAlignment=HAlign_Center, verticalAlignment=VAlign_Center, padding=<ltrb 4 2 4 2>}}
  }
  SuggestedActionLabelOverlay {
   widgetClassPath=&Overlay
   parent=ColumnLabelsHorizontalBox
   widget=SuggestedActionLabelOverlay
   slot {class=HorizontalBoxSlot, properties {padding=<ltrb 1 1 1 1>, size {sizeRule=Fill, value=1}}}
  }
  Image {
   widgetClassPath=&Image
   parent=SuggestedActionLabelOverlay
   widget=Image
   properties {colorAndOpacity=<rgba 0.0284 0.0284 0.0284 1>}
   slot {class=OverlaySlot, properties {horizontalAlignment=HAlign_Fill, verticalAlignment=VAlign_Fill}}
  }
  TextBlock_2 {
   widgetClassPath=&TextBlock
   parent=SuggestedActionLabelOverlay
   widget=TextBlock_2
   properties {
    text=Suggested Action
    colorAndOpacity {colorUseRule=UseColor_Specified, specifiedColor=<rgba 0.5271 0.5271 0.5271 1>}
    font {
     bForceMonospaced=false
     bMaterialIsStencil=false
     fontMaterial=null
     fontObject=&Roboto
     letterSpacing=0
     monospacedWidth=1
     size=9
     skewAmount=0
     typefaceFontName=Bold
     outlineSettings {
      bApplyOutlineToDropShadows=false
      bMiteredCorners=false
      bSeparateFillAlpha=false
      outlineMaterial=null
      outlineSize=0
      outlineColor=<rgba 0 0 0 1>
     }
    }
   }
   slot {class=OverlaySlot, properties {horizontalAlignment=HAlign_Center, verticalAlignment=VAlign_Center, padding=<ltrb 4 2 4 2>}}
  }
  EntriesScrollBox {
   widgetClassPath=&EditorUtilityScrollBox
   bIsVariable=true
   parent=ResultsVerticalBox
   widget=EntriesScrollBox
   slot {class=VerticalBoxSlot, properties {size {sizeRule=Fill, value=1}}}
  }
  ExporttoCSVButton {
   widgetClassPath=&EditorUtilityButton
   parent=ResultsVerticalBox
   widget=ExporttoCSVButton
   slot {class=VerticalBoxSlot, properties {horizontalAlignment=HAlign_Center}}
  }
  EditorUtilityEditableText_2 {
   widgetClassPath=&EditorUtilityEditableText
   parent=ExporttoCSVButton
   widget=EditorUtilityEditableText_2
   properties {
    isReadOnly=true
    justification=Center
    text=Export to CSV
    visibility=SelfHitTestInvisible
    widgetStyle {
     backgroundImageComposing {
      bIsDynamicallyLoaded=false
      drawAs=Border
      imageType=FullColor
      mirroring=NoMirror
      resourceName=../../../Engine/Content/Slate/Old/HyperlinkDotted.png
      resourceObject=null
      tiling=Both
      imageSize=<0 0>
      margin=<ltrb 0 0 0 0.1875>
      outlineSettings {
       bUseBrushTransparency=false
       roundingType=HalfHeightRadius
       width=0
       color {colorUseRule=UseColor_Specified, specifiedColor=<rgba 0 0 0 0>}
       cornerRadii=<0 0 0 0>
      }
      tintColor {colorUseRule=UseColor_Specified, specifiedColor=<rgba 1 1 1 1>}
      uVRegion {bIsValid=false, max=<0 0>, min=<0 0>}
     }
     backgroundImageSelected {
      bIsDynamicallyLoaded=false
      drawAs=Image
      imageType=NoImage
      mirroring=NoMirror
      resourceName=None
      resourceObject=null
      tiling=NoTile
      imageSize=<0 0>
      margin=<ltrb 0 0 0 0>
      outlineSettings {
       bUseBrushTransparency=false
       roundingType=HalfHeightRadius
       width=0
       color {colorUseRule=UseColor_Specified, specifiedColor=<rgba 0 0 0 0>}
       cornerRadii=<0 0 0 0>
      }
      tintColor {colorUseRule=UseColor_Specified, specifiedColor=<rgba 0 0.162 0.7454 1>}
      uVRegion {bIsValid=false, max=<0 0>, min=<0 0>}
     }
     caretImage {
      bIsDynamicallyLoaded=false
      drawAs=Image
      imageType=NoImage
      mirroring=NoMirror
      resourceName=None
      resourceObject=null
      tiling=NoTile
      imageSize=<0 0>
      margin=<ltrb 0 0 0 0>
      outlineSettings {
       bUseBrushTransparency=false
       roundingType=HalfHeightRadius
       width=0
       color {colorUseRule=UseColor_Specified, specifiedColor=<rgba 0 0 0 0>}
       cornerRadii=<0 0 0 0>
      }
      tintColor {colorUseRule=UseColor_Specified, specifiedColor=<rgba 1 1 1 1>}
      uVRegion {bIsValid=false, max=<0 0>, min=<0 0>}
     }
     colorAndOpacity {colorUseRule=UseColor_Specified, specifiedColor=<rgba 0.5271 0.5271 0.5271 1>}
     font {
      bForceMonospaced=false
      bMaterialIsStencil=false
      fontMaterial=null
      fontObject=null
      letterSpacing=0
      monospacedWidth=1
      size=9.75
      skewAmount=0
      typefaceFontName=Regular
      outlineSettings {
       bApplyOutlineToDropShadows=false
       bMiteredCorners=false
       bSeparateFillAlpha=false
       outlineMaterial=null
       outlineSize=0
       outlineColor=<rgba 0 0 0 1>
      }
     }
     hintColorAndOpacity {colorUseRule=UseColor_Foreground, specifiedColor=<rgba 1 0 1 1>}
     selectedBackgroundColor {colorUseRule=UseColor_Foreground, specifiedColor=<rgba 1 0 1 1>}
    }
   }
   slot {class=ButtonSlot, properties {horizontalAlignment=HAlign_Fill, verticalAlignment=VAlign_Fill, padding=<ltrb 5 5 5 5>}}
  }
 ]
}
