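# Editor utility widget in the DelMar track performance review tools that selects level actors using one exact material.
# A Details View bound to the widget itself exposes ExactMaterialToSearch as the only input.
# On button release, it scans every level actor's mesh components with GetMaterials, breaking at an actor's first match.
# It selects the collected actors in the editor, then clears the array for the next search.
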EditorUtilityWidgetBlueprint /DelMarTools/EditorUtilities/AssetReviewTools/TrackPerformanceReview/Subwidgets/FindBRDependencies/FindBRDependencies_SelectorByMat
graphs [EventGraph {nodes [
   "On Released (SearchButton)"() {is=title K2Node_ComponentBoundEvent}
   n1 = Get EditorActorSubsystem()
   > n2 = GetAllLevelActors(self=@n1)
   > n3 = For Each Loop with Break!(Exec:Exec, Array:Array of Actor Object References=@n2, Break:Exec) -> (LoopBody:Exec, Array Element:Actor Object Reference, Array Index:int, Completed:Exec)
   SetSelectedLevelActors(execute=@n3.Completed, self=@n1, ActorsToSelect=Get ActorsToSelect() -> (ActorsToSelect:Array of Actor Object References))
   > Clear(TargetArray:Array of Actor Object References=Get ActorsToSelect())
   n4 = For Each Loop with Break(
    Exec=@n3.LoopBody
    Array:Array of Mesh Component Object References=Get Components By Class!(self:Actor Object Reference=@n3, ComponentClass:Actor Component Class Reference=/Script/Engine.MeshComponent) -> (ReturnValue:Array of Mesh Component Object References)
    Break=@n7.then
   ) -> (Array Element:Mesh Component Object Reference)
   n5 = GetMaterials!(self:Mesh Component Object Reference=@n4) -> (ReturnValue:Array of Material Interface Object References)
   n6 = Get ExactMaterialToSearch() -> (ExactMaterialToSearch:Material Interface Object Reference)
   > Branch(Condition="integer >= integer"(A=Find Item!(TargetArray:Array of Material Interface Object References=@n5, ItemToFind:"Material Interface Object Reference (by ref)"=@n6) -> (ReturnValue)))
   > n7 = Add(TargetArray:Array of Actor Object References=Get ActorsToSelect(), NewItem:"Actor Object Reference (by ref)"=@n3) {is=title K2Node_CallArrayFunction}
   Event PreConstruct()
   > SetObject(self=Get ParamsView() -> (ParamsView:Details View Object Reference), NewObject=Self-Reference())
  ]}]
variables [
 ActorsToSelect {type="array<Actor>"}
 ExactMaterialToSearch {type=MaterialInterface, default=null}
]
widgets {
 info {rootWidgetClass=&VerticalBox, widgetCount=4}
 widgets [
  VerticalBox_0 {widgetClassPath=&VerticalBox, widget=VerticalBox_0}
  ParamsView {
   widgetClassPath=&DetailsView
   parent=VerticalBox_0
   widget=ParamsView
   properties {propertiesToShow [ExactMaterialToSearch]}
   slot {class=VerticalBoxSlot}
  }
  SearchButton {widgetClassPath=&EditorUtilityButton, parent=VerticalBox_0, widget=SearchButton, slot {class=VerticalBoxSlot}}
  CommonTextBlock_41 {
   widgetClassPath=&CommonTextBlock
   parent=SearchButton
   widget=CommonTextBlock_41
   properties {
    style=&TextStyle-Base_C
    text="Select matching actors (exact material)"
    font {
     bForceMonospaced=false
     bMaterialIsStencil=false
     fontMaterial=null
     fontObject=&NotoSans
     letterSpacing=0
     monospacedWidth=1
     size=24
     skewAmount=0
     typefaceFontName=NotoSans-Regular
     outlineSettings {
      bApplyOutlineToDropShadows=false
      bMiteredCorners=false
      bSeparateFillAlpha=false
      outlineMaterial=null
      outlineSize=0
      outlineColor=<rgba 0 0 0 1>
     }
    }
    shadowOffset=<0 0>
   }
   slot {class=ButtonSlot}
  }
 ]
}
